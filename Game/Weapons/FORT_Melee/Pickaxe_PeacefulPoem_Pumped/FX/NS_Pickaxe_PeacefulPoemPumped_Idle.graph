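# Idle Niagara effect for the PeacefulPoem Pumped pickaxe, keeping fire and a purple glow attached to the axe.
# FireSmall and FireSoft burst width-flickering flame meshes; FireDecals streams dim purple sprites that fade while drifting by curve-driven offsets.
# Particle partitions select between two IdleFX sockets, resampled every update, and vary flame material parameters; one pulsing light follows socket 02.
# A User.isBuiltIn comparison can zero every emitter's spawn count.
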NiagaraSystem /Game/Weapons/FORT_Melee/Pickaxe_PeacefulPoem_Pumped/FX/NS_Pickaxe_PeacefulPoemPumped_Idle
dependencies {
 usedDataInterfaces [&NiagaraDataInterfaceSkeletalMesh, &NiagaraDataInterfaceCurve]
 usedDynamicInputs []
 usedModules [&SystemState]
}
emitters {
 FireDecals {
  emitterUpdateScript [
   &EmitterState(Life Cycle Mode=::NewEnumerator1, Loop Duration=1, UseLoopDelay=false)
   &SpawnRate(
    SpawnRate=&MakeFloatFromInt(Int=&MakeCustomIntFromBool(
      Bool=&SetBoolByFloatComparison(A=1, B=$User.isBuiltIn, Comparison Type=::NewEnumerator4)
      True INT=0
      False INT=3
     ))
    Use Spawn Probability=false
   )
  ]
  particleSpawnScript [
   &InitializeParticle(
    Color=<rgba 0.0897 0 0.2188 1>
    Lifetime=4
    Sprite Size Mode=::NewEnumerator3
    Uniform Sprite Size=12
    UsePositionOffset=false
    Write Material Random=true
   )
   &PartitionParticles()
  ]
  particleUpdateScript [
   &ParticleState()
   &SkeletalMeshLocation(
    Bone / Socket Index=$Particles.PartitionParticles.Partition
    MaskSpawnGroup=false
    Mesh Sampling Type=::NewEnumerator5
    Sampled Position Offset=&MakeVector(
     X=&Lerp_Float(
      A=0
      B=40
      Alpha=&FloatFromCurve(FloatCurve=DataInterface() {propertyValues {Curve {keys [{tangentMode=RCTM_User, time=0, value=0}, {tangentMode=RCTM_User, time=1, value=1}]}}})
     )
     Y=0
     Z=&Lerp_Float(
      A=0
      B=10
      Alpha=&FloatFromCurve(FloatCurve=DataInterface() {propertyValues {Curve {keys [<at 0.6 0 RCIM_Cubic>, {arriveTangent=3.6383, tangentMode=RCTM_User, time=1, value=1}]}}})
     )
    )
    Skeletal Mesh=DataInterface() {propertyValues {
      DefaultMesh=/Game/Weapons/FORT_Melee/Pickaxe_PeacefulPoem_Pumped/Meshes/PeacefulPoem_Pumped_Axe
      ExcludeBoneName=root
      OptionalComponentClass=null
      PreviewMesh=&PeacefulPoem_Pumped_Axe
      SkinningMode=SkinOnTheFly
      SoftSourceActor=null
      SourceMode=Default
      UvSetIndex=0
      WholeMeshLOD=-1
      bExcludeBone=false
      bReadDeformedGeometry=true
      bRequireCurrentFrameData=true
      ComponentTags []
      FilteredBones []
      FilteredSockets [IdleFX_Socket_02, IdleFX_Socket_01]
      MeshUserParameter {parameter {name=None, typeDefHandle {registeredTypeIndex=1}, varData []}}
      SamplingRegions []
     }}
    Socket Sampling Mode=::NewEnumerator5
   )
   &CameraOffset(Camera Offset Amount=5)
   &DynamicMaterialParameters(
    Index 0 Param 1=1
    Index 0 Param 2=0.8
    Index 0 Param 3=&RandomRangeFloat(Minimum=0, Maximum=1)
    Index 0 Param 4=&RandomRangeFloat(Minimum=0, Maximum=1)
    Module.Param0WriteEnabled=true
    Module.Param1WriteEnabled=true
    Module.Param2WriteEnabled=true
    Module.Param3WriteEnabled=true
   ) {module=DynamicMaterialParameters001}
   &ScaleColor(
    Scale Alpha=&FloatFromCurve(
     FloatCurve=DataInterface() {propertyValues {Curve {keys [
         {interpMode=RCIM_Cubic, tangentMode=RCTM_User, time=0, value=0}
         {interpMode=RCIM_Cubic, tangentMode=RCTM_User, time=0.5, value=1}
         {tangentMode=RCTM_User, time=1, value=0}
        ]}}}
     Scale Curve=20
    )
    Scale RGB=&VectorFromFloat(Value=0.1)
    ScaleA=true
    ScaleRGB=true
   )
  ]
  rendererClasses [&NiagaraSpriteRendererProperties]
  renderers [{rendererClass=&NiagaraSpriteRendererProperties}]
 }
 FireSmall {
  emitterUpdateScript [
   &EmitterState(Life Cycle Mode=::NewEnumerator1, Loop Behavior=::NewEnumerator1, Loop Duration Mode=::NewEnumerator1, UseLoopDelay=false)
   &SpawnBurst_Instantaneous(
    Spawn Count=&MakeCustomIntFromBool(
     Bool=&SetBoolByFloatComparison(A=1, B=$User.isBuiltIn, Comparison Type=::NewEnumerator4)
     True INT=0
     False INT=2
    )
    Use Spawn Probability=false
    UseLoopCountLimit=false
   )
  ]
  particleSpawnScript [
   &InitializeParticle(Lifetime=6, Mesh Scale Mode=::NewEnumerator3, Sprite Size Mode=::NewEnumerator3, UsePositionOffset=true, Write Material Random=true)
   &Color(Color=&MakeLinearColorFromVectorAndFloat(
     "Vector (RGB)"=&MakeVector(
      X=&RandomRangeFloat(Minimum=0, Maximum=1)
      Y=&ReturnNormalizedExecIndex()
      Z=0
     )
     "Float (Alpha)"=1
    ))
   &InitialMeshOrientation(
    Mesh Axis To Orient=<0 0 1>
    Mesh Orientation Mode=::NewEnumerator2
    Orientation Vector=<0 0 1>
    Rotation Coordinate Space=::NewEnumerator2
    Use Rotation Vector=true
   )
   &PartitionParticles()
  ]
  particleUpdateScript [
   &ParticleState(Kill Particles When Lifetime Has Elapsed=false, Let Infinitely Lived Particles Die When Emitter Deactivates=true)
   &ScaleMeshSize(Scale Factor=&MakeVector(
     X=&Lerp_Float(
      A=0.8
      B=1
      Alpha=&Waveform(
       Global Input Position=&RandomRangeFloat(Minimum=0, Maximum=10)
       "Waveform [1]"=::NewEnumerator12
       "[1] Frequency"=1
       "[1] Amplitude Min/Max"=<0 1>
      )
     )
     Y=1
     Z=1
    ))
   &DynamicMaterialParameters(
    Index 0 Param 1=&Lerp_Float(A=0, B=0.5, Alpha=&MakeFloatFromInt(Int=$Particles.PartitionParticles.Partition))
    Index 0 Param 2=&Multiply_Float(A=&RandomRangeFloat(Minimum=1, Maximum=1.2), B=1.2)
    Module.Index0Param0Write=true
    Module.Index0Param1Write=true
    Module.Index0Param2Write=true
    Module.Index0Param3Write=true
   ) {module=DynamicMaterialParameters001}
   &SkeletalMeshLocation(
    Bone / Socket Index=$Particles.PartitionParticles.Partition
    MaskSpawnGroup=false
    Mesh Sampling Type=::NewEnumerator5
    Skeletal Mesh=DataInterface() {propertyValues {
      DefaultMesh=&PeacefulPoem_Pumped_Axe
      ExcludeBoneName=root
      OptionalComponentClass=null
      PreviewMesh=&PeacefulPoem_Pumped_Axe
      SkinningMode=SkinOnTheFly
      SoftSourceActor=null
      SourceMode=Default
      UvSetIndex=0
      WholeMeshLOD=-1
      bExcludeBone=false
      bReadDeformedGeometry=true
      bRequireCurrentFrameData=true
      ComponentTags []
      FilteredBones []
      FilteredSockets [IdleFX_Socket_02, IdleFX_Socket_01]
      MeshUserParameter {parameter {name=None, typeDefHandle {registeredTypeIndex=1}, varData []}}
      SamplingRegions []
     }}
    Socket Sampling Mode=::NewEnumerator5
   )
  ]
  rendererClasses [&NiagaraMeshRendererProperties]
  renderers [{rendererClass=&NiagaraMeshRendererProperties, rendererIndex=0}]
 }
 FireSoft {
  emitterUpdateScript [
   &EmitterState(Life Cycle Mode=::NewEnumerator1, Loop Behavior=::NewEnumerator1, Loop Duration Mode=::NewEnumerator1, UseLoopDelay=false)
   &SpawnBurst_Instantaneous(
    Spawn Count=&MakeCustomIntFromBool(
     Bool=&SetBoolByFloatComparison(A=1, B=$User.isBuiltIn, Comparison Type=::NewEnumerator4)
     True INT=0
     False INT=2
    )
    Use Spawn Probability=false
    UseLoopCountLimit=false
   )
  ]
  particleSpawnScript [
   &InitializeParticle(Lifetime=6, Mesh Scale Mode=::NewEnumerator3, Sprite Size Mode=::NewEnumerator3, UsePositionOffset=true, Write Material Random=true)
   &Color(Color=&MakeLinearColorFromVectorAndFloat(
     "Vector (RGB)"=&MakeVector(
      X=&RandomRangeFloat(Minimum=0, Maximum=1)
      Y=&ReturnNormalizedExecIndex()
      Z=0
     )
     "Float (Alpha)"=1
    ))
   &InitialMeshOrientation(
    Mesh Axis To Orient=<0 0 1>
    Mesh Orientation Mode=::NewEnumerator2
    Orientation Vector=<0 0 1>
    Rotation Coordinate Space=::NewEnumerator2
    Use Rotation Vector=true
   )
   &PartitionParticles()
  ]
  particleUpdateScript [
   &ParticleState(
    Kill Particles When Lifetime Has Elapsed=false
    Let Infinitely Lived Particles Die When Emitter Deactivates=true
    Loop Particles Lifetime=true
   )
   &ScaleMeshSize(Scale Factor=&MakeVector(
     X=&Lerp_Float(
      A=0.8
      B=1
      Alpha=&Waveform(
       Global Input Position=&RandomRangeFloat(Minimum=0, Maximum=10)
       "Waveform [1]"=::NewEnumerator12
       "[1] Frequency"=1
       "[1] Amplitude Min/Max"=<0 1>
      )
     )
     Y=1
     Z=1
    ))
   &DynamicMaterialParameters(
    Index 0 Param 1=&Lerp_Float(A=0, B=0.5, Alpha=&MakeFloatFromInt(Int=$Particles.PartitionParticles.Partition))
    Index 0 Param 2=&Multiply_Float(A=&RandomRangeFloat(Minimum=1, Maximum=1.2), B=1.2)
    Module.Index0Param0Write=true
    Module.Index0Param1Write=true
    Module.Index0Param2Write=true
    Module.Index0Param3Write=true
   ) {module=DynamicMaterialParameters001}
   &SkeletalMeshLocation(
    Bone / Socket Index=$Particles.PartitionParticles.Partition
    MaskSpawnGroup=false
    Mesh Sampling Type=::NewEnumerator5
    Skeletal Mesh=DataInterface() {propertyValues {
      DefaultMesh=&PeacefulPoem_Pumped_Axe
      ExcludeBoneName=root
      OptionalComponentClass=null
      PreviewMesh=&PeacefulPoem_Pumped_Axe
      SkinningMode=SkinOnTheFly
      SoftSourceActor=null
      SourceMode=Default
      UvSetIndex=0
      WholeMeshLOD=-1
      bExcludeBone=false
      bReadDeformedGeometry=true
      bRequireCurrentFrameData=true
      ComponentTags []
      FilteredBones []
      FilteredSockets [IdleFX_Socket_02, IdleFX_Socket_01]
      MeshUserParameter {parameter {name=None, typeDefHandle {registeredTypeIndex=1}, varData []}}
      SamplingRegions []
     }}
    Socket Sampling Mode=::NewEnumerator5
   )
  ]
  rendererClasses [&NiagaraMeshRendererProperties]
  renderers [{rendererClass=&NiagaraMeshRendererProperties, rendererIndex=0}]
 }
 Lights {
  emitterUpdateScript [
   &EmitterState(Life Cycle Mode=::NewEnumerator1, Loop Behavior=::NewEnumerator1, Loop Duration Mode=::NewEnumerator1, UseLoopDelay=false)
   &SpawnBurst_Instantaneous(
    Spawn Count=&MakeCustomIntFromBool(
     Bool=&SetBoolByFloatComparison(A=1, B=$User.isBuiltIn, Comparison Type=::NewEnumerator4)
     True INT=0
     False INT=1
    )
    Use Spawn Probability=false
    UseLoopCountLimit=false
   )
  ]
  particleSpawnScript [
   &InitializeParticle(
    Color=<rgba 0.5 0.1 0.8 1>
    Lifetime=6
    Mesh Scale Mode=::NewEnumerator3
    Sprite Size Mode=::NewEnumerator3
    UsePositionOffset=false
    Write Material Random=true
   )
   &PartitionParticles(Number Of Partitions=1)
  ]
  particleUpdateScript [
   &ParticleState(Kill Particles When Lifetime Has Elapsed=false, Let Infinitely Lived Particles Die When Emitter Deactivates=true)
   &SkeletalMeshLocation(
    Bone / Socket Index=$Particles.PartitionParticles.Partition
    MaskSpawnGroup=false
    Mesh Sampling Type=::NewEnumerator5
    Sampled Position Offset=<20 0 5>
    Skeletal Mesh=DataInterface() {propertyValues {
      DefaultMesh=&PeacefulPoem_Pumped_Axe
      ExcludeBoneName=root
      OptionalComponentClass=null
      PreviewMesh=&PeacefulPoem_Pumped_Axe
      SkinningMode=SkinOnTheFly
      SoftSourceActor=null
      SourceMode=Default
      UvSetIndex=0
      WholeMeshLOD=-1
      bExcludeBone=false
      bReadDeformedGeometry=true
      bRequireCurrentFrameData=true
      ComponentTags []
      FilteredBones []
      FilteredSockets [IdleFX_Socket_02]
      MeshUserParameter {parameter {name=None, typeDefHandle {registeredTypeIndex=1}, varData []}}
      SamplingRegions []
     }}
    Socket Sampling Mode=::NewEnumerator5
   )
   &FN_Light_Attributes(
    Color=<rgba 0.5 0.1 0.8 1>
    Create Color Attribute=true
    Create Exponent Attribute=true
    Create Position Attribute=true
    Create Radius Attribute=true
    Enable Particle Light Enable=true
    Exponent=2
    Light Enabled=true
    Position:NiagaraPosition=$Particles.Position
    Radius=15
    Scale Color=&Lerp_Float(
     A=10
     B=20
     Alpha=&Waveform("Waveform [1]"=::NewEnumerator12, "[1] Frequency"=3, "[1] Amplitude Min/Max"=<0 1>)
    )
   )
  ]
  rendererClasses [&NiagaraLightRendererProperties]
  renderers [{rendererClass=&NiagaraLightRendererProperties, rendererIndex=0}]
 }
}
summary {
 systemName=NS_Pickaxe_PeacefulPoemPumped_Idle
 emitters [
  {emitterName=FireSmall, rendererClasses [&NiagaraMeshRendererProperties]}
  {emitterName=Lights, rendererClasses [&NiagaraLightRendererProperties]}
  {emitterName=FireDecals, rendererClasses [&NiagaraSpriteRendererProperties]}
  {emitterName=FireSoft, rendererClasses [&NiagaraMeshRendererProperties]}
 ]
 userVariables [
  User.isBuiltIn {type=NiagaraFloat, defaultValue=0, description=""}
  User.RibbonIDOffset {type=NiagaraInt32}
 ]
}
